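# Material function for a performance-oriented transparent base material in AtomFramework, outputting a full material attributes set at fixed opacity.
# MF_PrimitiveColor feeds base, emissive and subsurface color with Fresnel shaping, while MF_PrimitiveSurface and MF_PrimitiveNormal supply surface properties and normal.
# A "Trans" reroute fakes see-through reflection and refraction by blending Fresnel-warped and plain skylight environment samples, tinted by a desaturated rough sample.
# A quality switch zeroes this emissive contribution on Low and Medium.
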MaterialFunction /AtomFramework/Materials/Game_Perf/MF_TranspBase
graph {nodes [
  n1 = MaterialFunctionCall!(Random Seed) {materialFunction=/AtomFramework/Materials/MatFunctions/MF_PrimitiveColor}
  n2 = Constant()
  n3 = QualitySwitch(Default=Multiply(A=NamedRerouteUsage()) {constB=0.2}, Low=@n2, Medium=@n2)
  n4 = Fresnel() {baseReflectFraction=0, exponent=1}
  n5 = SetMaterialAttributes!(
   MaterialAttributes
   BaseColor=Multiply(A=@n1.Result, B=Fresnel() {baseReflectFraction=0.4, exponent=1})
   EmissiveColor=Multiply(A=Multiply(A=@n1.Result, B=@n3), B=@n4)
   SubsurfaceColor=Multiply(A=Multiply(A=@n1.Result) {constB=0.25}, B=@n4)
   Opacity=Constant() {r=0.3}
  ) {attributeSetTypes [
    #1
    #2
    #3
    #4
   ]}
  n6 = MaterialFunctionCall!(Seed=Constant() {r=1}, Min Max Roughness=Constant2Vector() {g=0.1, r=0.05}, Metal, Spec=Constant() {r=0.6}, Edge Falloff) {materialFunction=/AtomFramework/Materials/MatFunctions/MF_PrimitiveSurface}
  n7 = SetMaterialAttributes!(
   MaterialAttributes=@n5
   Metallic=@n6.Metal
   Specular=@n6.Metal
   Roughness=@n6.Metal
   Normal=MaterialFunctionCall!(Flatten Edge warp, Flatten Injection Warp N, Random Seed) {materialFunction=/AtomFramework/Materials/MatFunctions/MF_PrimitiveNormal}.Result
  ) {attributeSetTypes [
    #5
    #6
    #7
    #8
   ]}
  FunctionOutput(None=@n7)
  n8 = Multiply(
   A=Fresnel(Normal=VertexNormalWS()) {baseReflectFraction=0, exponent=10}
   B=Transform(None=CameraVectorWS()) {transformSourceType=TRANSFORMSOURCE_World, transformType=TRANSFORM_Local}
  )
  n9 = Fresnel(Normal=VertexNormalWS()) {baseReflectFraction=0, exponent=4}
  n10 = Saturate(None=Abs(None=Subtract(A=@n9) {constB=0.4}))
  n11 = SkyLightEnvMapSample(Direction=Add(A=ReflectionVectorWS(CustomWorldNormal=@n8), B=Multiply(A=@n10) {constB=0.2}), Roughness=Constant())
  n12 = SkyLightEnvMapSample(Direction=ReflectionVectorWS())
  n13 = SkyLightEnvMapSample(Roughness=Constant() {r=1})
  n14 = Constant() {r=1}
  NamedRerouteDeclaration(None=Multiply(A=LinearInterpolate(A=@n11, B=@n12) {constAlpha=0.3}, B=Desaturation(None=@n13, Fraction=@n14)) {constB=0.5}) {name=Trans, nodeColor=<rgba 1 0.75 0 1>}
 ]}
